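# Material function for weapon skins that adds a faceted crystal look with fake depth, outputting crystal normals, luminance and UVs.
# A crack normal map's blue channel drives bump-offset parallax and an extra offset on tiled UVs for sampling the crystal texture.
# Its red-green channels form intensity-scaled normals blended with incoming normals through a mask; squared red gives masked luminance.
# Low quality skips the distortion and normal blend.
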MaterialFunction /Game/Weapons/WeaponSkins/MaterialFunction/MF_Crystal_Facetted
graph {
 properties {bExposeToLibrary=true}
 nodes [
  n1 = NamedRerouteUsage()
  n2 = TextureSample(UVs=Multiply(A=NamedRerouteUsage(), B=ScalarParameter() {defaultValue=6, group=Crystal_Facetted, parameterName=Crystal_CracksSize})) {
   samplerSource=SSM_Wrap_WorldGroupSettings
   samplerType=SAMPLERTYPE_Normal
   texture=/Game/Accessories/FORT_Backpacks/Backpack_F_MED_Punk_Koi_Summer/Textures/F_MED_Punk_Koi_Summer_Backpack_Crystal_N
  }
  n3 = ConstantBiasScale(None=Power(Base=@n2.B) {constExponent=4}) {bias=-0.5, scale=2}
  n4 = ScalarParameter() {defaultValue=0.05, group=Crystal_Facetted, parameterName=Crystal_CracksBelowOffset}
  n5 = NamedRerouteUsage()
  n6 = ScalarParameter() {defaultValue=3.7, group=Crystal_Facetted, parameterName=Crystal_Tiling}
  n7 = ScalarParameter() {defaultValue=2, group=Crystal_Facetted, parameterName=Crystal_DepthPower}
  n8 = Add(
   A=Add(A=Multiply(A=@n3, B=@n4), B=Multiply(A=@n5, B=@n6))
   B=MaterialFunctionCall!(
    HeightTexture=Power(Base=@n2.B, Exp=@n7)
    TextureCoordinates
    ReferencePlane=ScalarParameter() {defaultValue=1, group=Crystal_Facetted, parameterName=Crystal_Height}
    HeightRatioInput=ScalarParameter() {defaultValue=0.25, group=Crystal_Facetted, parameterName=Crystal_HeightRatio}
    CameraVector
   ) {materialFunction=/Engine/Functions/Engine_MaterialFunctions02/Utility/BumpOffset_Advanced.BumpOffset_advanced}.Distortion
  )
  n9 = Multiply(A=@n5, B=ScalarParameter() {defaultValue=3.7, group=Crystal_Facetted, parameterName=Crystal_Tiling_LOW})
  n10 = TextureSample(UVs=NamedRerouteDeclaration(None=QualitySwitch(Default=@n8, Low=@n9)) {name=CrystalUVs, nodeColor=<rgba 1 0.5156 0 1>}) {
   samplerSource=SSM_Wrap_WorldGroupSettings
   samplerType=SAMPLERTYPE_LinearColor
   texture=/Game/Characters/Player/Female/Medium/Bodies/F_MED_Shapeshifter/Textures/T_Crystal_Normal
  }
  FunctionOutput(None=Multiply(A=@n1, B=Multiply(A=@n10.R, B=@n10.R))) {outputName=Crystal Luminance, sortPriority=20}
  n11 = NamedRerouteUsage()
  n12 = ComponentMask(None=@n10.RGB) {g=true, r=true}
  n13 = Saturate(None=NamedRerouteUsage())
  n14 = MaterialFunctionCall(
   BaseNormal=@n11
   AdditionalNormal=MaterialFunctionCall(
    Normal=MaterialFunctionCall!(XY vector=@n12, Z sign, Default Normalization vector) {materialFunction=/Engine/Functions/Engine_MaterialFunctions02/Math/DeriveNormalZ_Function}.Result
    Flatness=OneMinus(None=@n13)
   ) {materialFunction=&FlattenNormal}.Result
  ) {materialFunction=&BlendAngleCorrectedNormals}
  FunctionOutput(None=QualitySwitch(Default=LinearInterpolate(A=@n11, B=@n14.Result, Alpha=@n1), Low=@n11)) {outputName=Crystal Normals, sortPriority=10}
  FunctionOutput(None=NamedRerouteUsage()) {outputName=Crystal UVs}
  NamedRerouteDeclaration(None=FunctionInput(Preview=TextureCoordinate()) {bUsePreviewValueAsDefault=true, inputName=InUVs, inputType=FunctionInput_Vector2}) {name=UVs, nodeColor=<rgba 1 0.1406 0 1>}
  NamedRerouteDeclaration(None=FunctionInput(Preview=Constant() {r=1}) {bUsePreviewValueAsDefault=true, inputName=Mask, inputType=FunctionInput_Scalar, sortPriority=20}) {name=Mask, nodeColor=<rgba 0.9219 1 0 1>}
  NamedRerouteDeclaration(None=FunctionInput(Preview=Constant() {r=1}) {bUsePreviewValueAsDefault=true, inputName=NormalIntensity, sortPriority=15}) {name=NormalIntensity, nodeColor=<rgba 1 0 0.375 1>}
  NamedRerouteDeclaration(None=FunctionInput(Preview=Constant3Vector() {constant=<rgba 0 0 1 0>}) {bUsePreviewValueAsDefault=true, inputName=InNormals, sortPriority=10}) {name=InNormals, nodeColor=<rgba 0.3125 0 1 1>}
 ]
}
