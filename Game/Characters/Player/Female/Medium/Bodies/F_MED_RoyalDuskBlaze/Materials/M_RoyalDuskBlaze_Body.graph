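# Body material for the RoyalDuskBlaze female character, adding animated faceted-crystal shimmer to the emissive between the 2023 character parent PreFX and PostFX functions.
# Scrolling noise drives crystal normal intensity, luminance contrast and brightness.
# A mask from the Emissive blue channel, or an earring UV mask via the IsFaceAcc switch, limits the crystal effect.
# Time-of-day scalars weight the shimmer, which multiplies the parent emissive, while crystal normals replace its normal.
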Material /Game/Characters/Player/Female/Medium/Bodies/F_MED_RoyalDuskBlaze/Materials/M_RoyalDuskBlaze_Body
graph {
 properties {
  bUseMaterialAttributes=true
  bUsedWithMeshDeformer=true
  bUsedWithMorphTargets=true
  bUsedWithNanite=true
  bUsedWithSkeletalMesh=true
  pixelDepthOffsetMode=PDOM_Legacy
  shadingModel=MSM_FromMaterialExpression
  usedShadingModels=Default Lit | Subsurface | Clear Coat | Subsurface Profile | Cloth
 }
 nodes [
  n1 = MaterialFunctionCall() {materialFunction=&MF_CharacterParent_2023_PreFX}
  n2 = GetMaterialAttributes(None=@n1.Result) {attributeGetTypes [#1, #2]}
  n3 = Multiply(A=TextureCoordinate(), B=ScalarParameter() {defaultValue=8, group=RoyalDusk, parameterName=Noise Tiling})
  n4 = ScalarParameter() {defaultValue=0.35, group=RoyalDusk, parameterName=Noise Speed}
  n5 = Multiply(A=Time(), B=AppendVector(A=Divide(A=@n4) {constB=4}, B=@n4)) {constB=0.3}
  n6 = Power(
   Base=TextureSample(UVs=Subtract(A=@n3, B=@n5)) {texture=/Game/Effects/Fort_Effects/Emotes/Candor/T_Noise_Blur_Texture}.RGB
   Exp=ScalarParameter() {defaultValue=1.6, group=RoyalDusk, parameterName=Noise Power}
  ) {constExponent=1.6}
  n7 = TextureCoordinate()
  n8 = Constant2Vector() {g=0.051, r=-0.16}
  n9 = TextureSampleParameter2D() {group=06_Emissive, parameterName=Emissive, sortPriority=1, texture=&T_BlacksRGB}
  n10 = StaticSwitchParameter(
   True=MaterialFunctionCall(UVs=Add(A=@n7, B=@n8), Texture Scale=Constant() {r=0.076}) {desc=earring mask, materialFunction=&ScaleUVsByCenter}.0-1 mask
   False=MaterialFunctionCall(In=Power(Base=@n9.B) {constExponent=0.5}, Contrast=Constant() {r=0.1}) {materialFunction=&CheapContrast}.Result
  ) {group=RoyalDusk, parameterName=IsFaceAcc}
  n11 = MaterialFunctionCall!(
   InUVs
   InNormals=@n2.Normal
   NormalIntensity=Add(A=ScalarParameter() {defaultValue=0.5, group=Crystal_Facetted, parameterName=Crystal_BaseNormalIntensity}, B=@n6)
   Mask=@n10
  ) {materialFunction=/Game/Weapons/WeaponSkins/MaterialFunction/MF_Crystal_Facetted}
  n12 = ScalarParameter() {group=Crystal_Facetted, parameterName=Crystal_Contrast}
  n13 = Multiply(A=ScalarParameter() {defaultValue=1, group=RoyalDusk, parameterName=Noise Contrast Bias}, B=@n6)
  n14 = MaterialFunctionCall(In=@n11.Crystal Luminance, Contrast=Multiply(A=Add(A=@n12, B=@n13), B=@n10)) {materialFunction=&CheapContrast}
  n15 = Multiply(A=@n6, B=ScalarParameter() {defaultValue=1, group=RoyalDusk, parameterName=Noise Intensity})
  n16 = ScalarParameter() {defaultValue=1, group=RoyalDusk, parameterName=TOD_Morning}
  n17 = ScalarParameter() {defaultValue=1, group=RoyalDusk, parameterName=TOD_Day}
  n18 = ScalarParameter() {defaultValue=1, group=RoyalDusk, parameterName=TOD_Evening}
  n19 = ScalarParameter() {defaultValue=1, group=RoyalDusk, parameterName=TOD_Night}
  n20 = Add(A=Multiply(
    A=Multiply(A=@n14.Result, B=@n15)
    B=MaterialFunctionCall(Value_Morning=@n16, Value_Day=@n17, Value_Evening=@n18, Value_Night=@n19) {materialFunction=&Scalar_Time_of_Day}.Result
   ) {constB=0})
  MaterialFunctionCall(
   Material=SetMaterialAttributes!(MaterialAttributes=@n1.Result, Emissive Color=Multiply(A=@n2.Emissive Color, B=@n20), Normal=@n11.Crystal Normals) {attributeSetTypes [#1, #2]}
   M=@n1.M
   Pre Skinned Local Position=@n1.M
   Distance Blend=@n1.M
   Fresnel=@n1.M
  ) {materialFunction=&MF_CharacterParent_2023_PostFX}
 ]
}
